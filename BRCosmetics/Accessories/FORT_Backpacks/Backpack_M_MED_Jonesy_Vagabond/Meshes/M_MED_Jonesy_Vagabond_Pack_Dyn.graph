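# Physics asset for the Jonesy Vagabond backpack cosmetic, giving capsule bodies and joint limits to its dynamic tongue, eye, olive, cheese and top bones.
# A three-capsule pack body hangs off root and carries the top, both cheese pieces and a three-link tongue chain.
# The eyes and olive attach to the top.
# Pack and top joints have tighter, asymmetric swing limits than the other parts, and every joint allows little twist.
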PhysicsAsset /BRCosmetics/Accessories/FORT_Backpacks/Backpack_M_MED_Jonesy_Vagabond/Meshes/M_MED_Jonesy_Vagabond_Pack_Dyn
bodies {
 dyn_cheese_l [{shapeType=Capsule, length=2.592, radius=0.6537, center=<2 0 0>, rotation=<rot 90 0 0>}]
 dyn_cheese_r [{shapeType=Capsule, length=2.592, radius=0.6537, center=<2 0 0>, rotation=<rot 90 0 0>}]
 dyn_eye_l [{shapeType=Capsule, length=2.592, radius=1.1537, center=<0.2 0 0>, rotation=<rot 90 0 0>}]
 dyn_eye_r [{shapeType=Capsule, length=2.592, radius=1.1537, center=<0.2 0 0>, rotation=<rot 90 0 0>}]
 dyn_olive [{shapeType=Capsule, length=2.092, radius=1.1537, center=<0.2 0 0>, rotation=<rot 90 0 0>}]
 dyn_tongue_01 [{shapeType=Capsule, length=3.842, radius=1.4037, center=<3 0 0.6>, rotation=<rot 90 0 -90>}]
 dyn_tongue_02 [{shapeType=Capsule, length=3.342, radius=1.4037, center=<3 0 0>, rotation=<rot 90 -165.9638 104.0362>}]
 dyn_tongue_03 [{shapeType=Capsule, length=3.342, radius=1.4037, center=<4 0 0>, rotation=<rot 90 0 -90>}]
 dyn_top [{shapeType=Capsule, length=1.842, radius=2.4037, center=<3 0 0>, rotation=<rot 90 0 0>}]
 pack [
  {shapeType=Capsule, length=2.25, radius=2.7812, shapeName=base, rotation=<rot 0 0 90>}
  {shapeType=Capsule, length=6.625, radius=2.75, shapeName=coll 1, center=<0.3709 -11.9352 -0.8578>, rotation=<rot -88.1881 -0.0004 6.7124>}
  {
   shapeType=Capsule
   length=7.9062
   radius=3.1562
   shapeName=coll 2
   center=<0.3182 -7.7438 -1.9874>
   rotation=<rot 85.9228 -159.9956 179.9998>
  }
 ]
 root [{shapeType=Capsule, length=0.1, radius=0.12}]
}
constraints [
 {bone1Name=dyn_tongue_02, bone2Name=dyn_tongue_01, swing1LimitDegrees=30, swing2LimitDegrees=30, twistLimitDegrees=5}
 {bone1Name=dyn_tongue_03, bone2Name=dyn_tongue_02, swing1LimitDegrees=30, swing2LimitDegrees=30, twistLimitDegrees=5}
 {bone1Name=dyn_olive, bone2Name=dyn_top, swing1LimitDegrees=30, swing2LimitDegrees=30, twistLimitDegrees=5}
 {bone1Name=dyn_eye_l, bone2Name=dyn_top, swing1LimitDegrees=30, swing2LimitDegrees=30, twistLimitDegrees=5}
 {bone1Name=dyn_eye_r, bone2Name=dyn_top, swing1LimitDegrees=30, swing2LimitDegrees=30, twistLimitDegrees=5}
 {bone1Name=dyn_tongue_01, bone2Name=pack, swing1LimitDegrees=30, swing2LimitDegrees=30, twistLimitDegrees=5}
 {bone1Name=dyn_cheese_l, bone2Name=pack, swing1LimitDegrees=30, swing2LimitDegrees=30, twistLimitDegrees=5}
 {bone1Name=dyn_top, bone2Name=pack, swing1LimitDegrees=5, swing2LimitDegrees=20, twistLimitDegrees=5}
 {bone1Name=dyn_cheese_r, bone2Name=pack, swing1LimitDegrees=30, swing2LimitDegrees=30, twistLimitDegrees=5}
 {bone1Name=pack, bone2Name=root, swing1LimitDegrees=5, swing2LimitDegrees=20, twistLimitDegrees=5}
]
